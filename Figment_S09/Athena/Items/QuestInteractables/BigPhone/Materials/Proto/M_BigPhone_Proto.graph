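# M_BigPhone_Proto is a prototype Default Lit material for the BigPhone quest interactable.
# It holds only a constant blue color vector and three scalar constants, with no textures, parameters or math nodes.
# The graph lists no links from these constants to material outputs, so their roles are not shown.
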Material /Figment_S09/Athena/Items/QuestInteractables/BigPhone/Materials/Proto/M_BigPhone_Proto
graph {
 properties {pixelDepthOffsetMode=PDOM_Legacy, usedShadingModels=Default Lit}
 nodes [
  Constant3Vector() {constant=<rgba 0.0953 0.3474 0.6458 0>}
  Constant() {r=0.6}
  Constant() {r=0.1}
  Constant() {r=0.2}
 ]
}
